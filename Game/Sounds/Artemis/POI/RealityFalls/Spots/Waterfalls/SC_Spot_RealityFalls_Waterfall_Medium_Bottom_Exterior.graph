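# Looping ambient cue for the exterior bottom of a medium waterfall at the Reality Falls POI, in the water ambient sound class.
# A quality switch selects either a three-layer mix of main loop, close loop and pitched-up distant roar, or a single pitch-modulated close roar for each other tier.
# Each layer has its own occluded attenuation, with the distant roar's custom curve swelling then fading over distance.
# Cue-level overrides cap concurrency and send to distant reverb.
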SoundCue /Game/Sounds/Artemis/POI/RealityFalls/Spots/Waterfalls/SC_Spot_RealityFalls_Waterfall_Medium_Bottom_Exterior
root=QualityLevel_0
nodes [
 {
  id=Attenuation_0
  children [WavePlayer_2]
  properties {
   bOverrideAttenuation=true
   attenuationOverrides {
    bAttenuateWithLPF=true
    bEnableFocusInterpolation=true
    bEnableListenerFocus=true
    bEnableOcclusion=true
    distanceAlgorithm=Logarithmic
    falloffDistance=7500
    focusAttackInterpSpeed=0.2
    lPFFrequencyAtMax=1000
    lPFRadiusMax=4000
    lPFRadiusMin=1000
    nonFocusVolumeAttenuation=0.8
    nonSpatializedRadiusStart=3000
    occlusionInterpolationTime=2
    occlusionLowPassFilterFrequency=1000
    occlusionTraceChannel=ECC_GameTraceChannel11
    stereoSpread=500
    attenuationShapeExtents {x=2000}
   }
  }
 }
 {
  id=Attenuation_1
  children [WavePlayer_1]
  properties {
   bOverrideAttenuation=true
   attenuationOverrides {
    bAttenuateWithLPF=true
    bEnableFocusInterpolation=true
    bEnableListenerFocus=true
    bEnableOcclusion=true
    distanceAlgorithm=Logarithmic
    falloffDistance=1500
    focusAttackInterpSpeed=0.2
    lPFFrequencyAtMax=1000
    lPFRadiusMax=4000
    lPFRadiusMin=1000
    nonFocusVolumeAttenuation=0.8
    nonSpatializedRadiusStart=3000
    occlusionInterpolationTime=2
    occlusionLowPassFilterFrequency=1000
    occlusionTraceChannel=ECC_GameTraceChannel11
    stereoSpread=500
    attenuationShapeExtents {x=750}
   }
  }
 }
 {
  id=Attenuation_2
  children [Modulator_4]
  properties {
   bOverrideAttenuation=true
   attenuationOverrides {
    bEnableFocusInterpolation=true
    bEnableListenerFocus=true
    bEnableOcclusion=true
    distanceAlgorithm=Custom
    falloffDistance=7500
    lPFFrequencyAtMax=1000
    lPFRadiusMax=4000
    lPFRadiusMin=1000
    nonFocusVolumeAttenuation=0.8
    nonSpatializedRadiusStart=5000
    occlusionInterpolationTime=3
    occlusionLowPassFilterFrequency=1000
    occlusionTraceChannel=ECC_GameTraceChannel11
    stereoSpread=10000
    attenuationShapeExtents {x=2000}
    customAttenuationCurve {editorCurveData {keys [
       <at 0 0>
       {arriveTangent=0.028, interpMode=RCIM_Cubic, leaveTangent=0.028, tangentMode=RCTM_User, time=0.2209, value=1}
       {arriveTangent=-0.2554, interpMode=RCIM_Cubic, leaveTangent=-0.2554, tangentMode=RCTM_User, time=0.4, value=0.97}
       <at 1 0>
      ]}}
   }
  }
 }
 {
  id=Attenuation_3
  children [Modulator_0]
  properties {
   bOverrideAttenuation=true
   attenuationOverrides {
    bAttenuateWithLPF=true
    bEnableFocusInterpolation=true
    bEnableListenerFocus=true
    bEnableOcclusion=true
    distanceAlgorithm=Logarithmic
    falloffDistance=8500
    lPFFrequencyAtMax=1000
    lPFRadiusMax=4000
    lPFRadiusMin=1000
    nonFocusVolumeAttenuation=0.8
    nonSpatializedRadiusStart=3000
    occlusionInterpolationTime=2
    occlusionLowPassFilterFrequency=1000
    occlusionTraceChannel=ECC_GameTraceChannel11
    stereoSpread=500
    attenuationShapeExtents {x=1900}
   }
  }
 }
 {
  id=Attenuation_4
  children [Modulator_1]
  properties {
   bOverrideAttenuation=true
   attenuationOverrides {
    bAttenuateWithLPF=true
    bEnableFocusInterpolation=true
    bEnableListenerFocus=true
    bEnableOcclusion=true
    distanceAlgorithm=Logarithmic
    falloffDistance=7500
    lPFFrequencyAtMax=1000
    lPFRadiusMax=4000
    lPFRadiusMin=1000
    nonFocusVolumeAttenuation=0.8
    nonSpatializedRadiusStart=3000
    occlusionInterpolationTime=2
    occlusionLowPassFilterFrequency=1000
    occlusionTraceChannel=ECC_GameTraceChannel11
    stereoSpread=500
    attenuationShapeExtents {x=1900}
   }
  }
 }
 {
  id=Attenuation_5
  children [Modulator_3]
  properties {
   bOverrideAttenuation=true
   attenuationOverrides {
    bAttenuateWithLPF=true
    bEnableFocusInterpolation=true
    bEnableListenerFocus=true
    bEnableOcclusion=true
    distanceAlgorithm=Logarithmic
    falloffDistance=7500
    lPFFrequencyAtMax=1000
    lPFRadiusMax=4000
    lPFRadiusMin=1000
    nonFocusVolumeAttenuation=0.8
    nonSpatializedRadiusStart=3000
    occlusionInterpolationTime=2
    occlusionLowPassFilterFrequency=1000
    occlusionTraceChannel=ECC_GameTraceChannel11
    stereoSpread=500
    attenuationShapeExtents {x=1900}
   }
  }
 }
 {id=Mixer_0, children [Attenuation_0, Attenuation_1, Attenuation_2], properties {inputVolume [3,0.5,1]}}
 {id=Modulator_0, children [WavePlayer_4], properties {pitchMax=1.3, pitchMin=1.2}}
 {id=Modulator_1, children [WavePlayer_5], properties {pitchMax=1.1, pitchMin=0.9}}
 {id=Modulator_3, children [WavePlayer_9], properties {pitchMax=1.1, pitchMin=0.9}}
 {id=Modulator_4, children [WavePlayer_3], properties {pitchMax=1.3, pitchMin=1.2}}
 {id=QualityLevel_0, children [Mixer_0, Attenuation_3, Attenuation_4, Attenuation_5]}
 {
  id=WavePlayer_1
  properties {
   bLooping=true
   soundWaveAssetPtr=&/Ambience_RealityFalls_Waterfall_Medium_Bottom_Close_Exterior_Loop
  }
 }
 {
  id=WavePlayer_2
  properties {
   bLooping=true
   soundWaveAssetPtr=&/Ambience_RealityFalls_Waterfall_Medium_Bottom_Exterior_Loop
  }
 }
 {
  id=WavePlayer_3
  properties {bLooping=true, soundWaveAssetPtr=&/Spot_RealityFalls_Waterfall_Roar_Medium_Distant_01}
 }
 {
  id=WavePlayer_4
  properties {bLooping=true, soundWaveAssetPtr=&/Spot_RealityFalls_Waterfall_Roar_Close}
 }
 {
  id=WavePlayer_5
  properties {bLooping=true, soundWaveAssetPtr=&Spot_RealityFalls_Waterfall_Roar_Close}
 }
 {
  id=WavePlayer_9
  properties {bLooping=true, soundWaveAssetPtr=&Spot_RealityFalls_Waterfall_Roar_Close}
 }
]
properties {
 bDebug=true
 bOverrideAttenuation=true
 bOverrideConcurrency=true
 maxDistance=10400
 soundClassObject=/Game/Sounds/FortSoundClassesAndModes/Fort_Water_Ambient
 volumeMultiplier=1
 attenuationOverrides {
  bAttenuateWithLPF=true
  bEnableListenerFocus=true
  bEnableOcclusion=true
  bEnableReverbSend=false
  bEnableSubmixSends=true
  distanceAlgorithm=Logarithmic
  falloffDistance=1500
  hPFFrequencyAtMax=200
  lPFFrequencyAtMax=1000
  lPFRadiusMax=2000
  lPFRadiusMin=1000
  nonFocusVolumeAttenuation=0.5
  nonSpatializedRadiusMode=Direct2D
  occlusionInterpolationTime=0.5
  occlusionLowPassFilterFrequency=15000
  occlusionTraceChannel=ECC_GameTraceChannel15
  occlusionVolumeAttenuation=0.3
  attenuationShapeExtents {x=500}
  submixSendSettings [{
    bEnableHPFCutoff=false
    bEnableLPFCutoff=false
    disableManualSendClamp=false
    hPFCutoff=20
    lPFCutoff=20000
    maxSendDistance=3000
    maxSendLevel=0.5
    minSendDistance=400
    minSendLevel=0.2
    sendLevel=0.2
    sendLevelControlMethod=Linear
    soundSubmix=&Fortnite_Reverb_Distant
   }]
 }
 concurrencyOverrides {maxCount=3}
}
