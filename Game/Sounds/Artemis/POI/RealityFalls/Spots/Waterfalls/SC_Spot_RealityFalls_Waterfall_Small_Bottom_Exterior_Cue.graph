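# Looping spot ambience cue for the exterior bottom of a small waterfall at the Reality Falls POI, in the water ambient sound class.
# A mixer blends two similarly named loop players, each with its own attenuation: a shorter natural-falloff layer and a longer logarithmic layer through a modulator.
# Both layers use listener focus, distance low-pass filtering and occlusion settings.
# All quality levels share that mixer, and concurrency allows one instance.
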SoundCue /Game/Sounds/Artemis/POI/RealityFalls/Spots/Waterfalls/SC_Spot_RealityFalls_Waterfall_Small_Bottom_Exterior_Cue
root=QualityLevel_0
nodes [
 {
  id=Attenuation_4
  children [WavePlayer_1]
  properties {
   bOverrideAttenuation=true
   attenuationOverrides {
    bAttenuateWithLPF=true
    bEnableFocusInterpolation=true
    bEnableListenerFocus=true
    distanceAlgorithm=NaturalSound
    falloffDistance=2000
    focusAttackInterpSpeed=0.5
    lPFFrequencyAtMax=1000
    lPFRadiusMax=4000
    lPFRadiusMin=1000
    nonFocusVolumeAttenuation=0.3
    nonSpatializedRadiusStart=500
    occlusionInterpolationTime=2
    occlusionLowPassFilterFrequency=1000
    occlusionTraceChannel=ECC_GameTraceChannel11
    reverbWetLevelMin=0.4
    attenuationShapeExtents {x=500}
   }
  }
 }
 {
  id=Attenuation_5
  children [Modulator_2]
  properties {
   bOverrideAttenuation=true
   attenuationOverrides {
    bAttenuateWithLPF=true
    bEnableFocusInterpolation=true
    bEnableListenerFocus=true
    distanceAlgorithm=Logarithmic
    falloffDistance=4000
    focusAttackInterpSpeed=0.2
    focusVolumeAttenuation=1.2
    lPFFrequencyAtMax=4000
    lPFRadiusMax=4000
    lPFRadiusMin=1000
    nonFocusVolumeAttenuation=0.8
    nonSpatializedRadiusStart=500
    occlusionInterpolationTime=2
    occlusionLowPassFilterFrequency=1000
    occlusionTraceChannel=ECC_GameTraceChannel11
    reverbWetLevelMin=0.4
    attenuationShapeExtents {x=1000}
   }
  }
 }
 {id=Mixer_1, children [Attenuation_5, Attenuation_4], properties {inputVolume [3,3]}}
 {id=Modulator_2, children [WavePlayer_0], properties {pitchMax=1, pitchMin=1}}
 {id=QualityLevel_0, children [Mixer_1,Mixer_1,Mixer_1,Mixer_1]}
 {
  id=WavePlayer_0
  properties {
   bLooping=true
   soundWaveAssetPtr=&/Ambience_RealityFalls_Waterfall_Small_Bottom_Exterior_Loop
  }
 }
 {
  id=WavePlayer_1
  properties {
   bLooping=true
   soundWaveAssetPtr=&Ambience_RealityFalls_Waterfall_Small_Bottom_Exterior_Loop
  }
 }
]
properties {
 bDebug=true
 bOverrideAttenuation=true
 bOverrideConcurrency=true
 maxDistance=5000
 soundClassObject=/Game/Sounds/FortSoundClassesAndModes/Fort_Water_Ambient
 volumeMultiplier=1
 attenuationOverrides {
  bEnableFocusInterpolation=true
  bEnableListenerFocus=true
  distanceAlgorithm=Logarithmic
  falloffDistance=3500
  focusAttackInterpSpeed=0.2
  focusVolumeAttenuation=1.2
  nonFocusVolumeAttenuation=0.8
  reverbWetLevelMin=0.4
  attenuationShapeExtents {x=1000}
 }
 concurrencyOverrides {maxCount=1}
}
