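# Looping spot ambience cue for the exterior top of a small Reality Falls waterfall, in the water ambient sound class.
# A mixer blends the medium waterfall-top loop, through short and mid-range logarithmic attenuations, with a quieter large waterfall-bottom loop on a longer custom curve.
# Each layer overrides attenuation with distance low-pass filtering and listener focus, but only the mid-range layer enables occlusion.
# A concurrency override limits simultaneous instances.
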SoundCue /Game/Sounds/Artemis/POI/RealityFalls/Spots/Waterfalls/SC_Spot_RealityFalls_Waterfall_Small_Top_Exterior_Cue
root=Mixer_0
nodes [
 {
  id=Attenuation_0
  children [WavePlayer_1]
  properties {
   bOverrideAttenuation=true
   attenuationOverrides {
    bAttenuateWithLPF=true
    bEnableFocusInterpolation=true
    bEnableListenerFocus=true
    bEnableOcclusion=true
    distanceAlgorithm=Logarithmic
    falloffDistance=5000
    focusAttackInterpSpeed=0.2
    focusVolumeAttenuation=1.2
    lPFFrequencyAtMax=1000
    lPFRadiusMax=5000
    lPFRadiusMin=2000
    nonFocusAzimuth=75
    nonFocusVolumeAttenuation=0.8
    nonSpatializedRadiusStart=2000
    occlusionInterpolationTime=2
    occlusionLowPassFilterFrequency=1000
    occlusionTraceChannel=ECC_GameTraceChannel11
    occlusionVolumeAttenuation=0.5
    stereoSpread=2000
    attenuationShapeExtents {x=750}
    customAttenuationCurve {editorCurveData {keys [
       <at 0 0>
       {arriveTangent=0.028, interpMode=RCIM_Cubic, leaveTangent=0.028, tangentMode=RCTM_User, time=0.2209, value=1}
       {arriveTangent=-0.2554, interpMode=RCIM_Cubic, leaveTangent=-0.2554, tangentMode=RCTM_User, time=0.4, value=0.97}
       <at 1 0>
      ]}}
   }
  }
 }
 {
  id=Attenuation_1
  children [WavePlayer_1]
  properties {
   bOverrideAttenuation=true
   attenuationOverrides {
    bAttenuateWithLPF=true
    bEnableFocusInterpolation=true
    bEnableListenerFocus=true
    distanceAlgorithm=Logarithmic
    falloffDistance=1000
    focusAttackInterpSpeed=0.2
    focusVolumeAttenuation=1.2
    lPFFrequencyAtMax=1000
    lPFRadiusMax=5000
    lPFRadiusMin=2000
    nonFocusAzimuth=75
    nonFocusVolumeAttenuation=0.8
    nonSpatializedRadiusStart=2000
    occlusionInterpolationTime=2
    occlusionLowPassFilterFrequency=1000
    occlusionTraceChannel=ECC_GameTraceChannel11
    occlusionVolumeAttenuation=0.2
    stereoSpread=2000
    attenuationShapeExtents {x=750}
    customAttenuationCurve {editorCurveData {keys [
       <at 0 0>
       {arriveTangent=0.028, interpMode=RCIM_Cubic, leaveTangent=0.028, tangentMode=RCTM_User, time=0.2209, value=1}
       {arriveTangent=-0.2554, interpMode=RCIM_Cubic, leaveTangent=-0.2554, tangentMode=RCTM_User, time=0.4, value=0.97}
       <at 1 0>
      ]}}
   }
  }
 }
 {
  id=Attenuation_2
  children [WavePlayer_0]
  properties {
   bOverrideAttenuation=true
   attenuationOverrides {
    bAttenuateWithLPF=true
    bEnableFocusInterpolation=true
    bEnableListenerFocus=true
    distanceAlgorithm=Custom
    falloffDistance=7500
    focusAttackInterpSpeed=2
    focusAzimuth=45
    focusReleaseInterpSpeed=2
    hPFFrequencyAtMax=500
    lPFFrequencyAtMax=1000
    lPFFrequencyAtMin=12000
    lPFRadiusMax=2000
    lPFRadiusMin=1000
    nonFocusAzimuth=110
    nonFocusVolumeAttenuation=0.5
    nonSpatializedRadiusStart=1000
    occlusionInterpolationTime=3
    occlusionLowPassFilterFrequency=1000
    occlusionTraceChannel=ECC_GameTraceChannel11
    attenuationShapeExtents {x=1000}
    customAttenuationCurve {editorCurveData {keys [
       <at 0 0.75>
       {arriveTangent=0.028, interpMode=RCIM_Cubic, leaveTangent=0.028, tangentMode=RCTM_User, time=0.2209, value=1}
       {arriveTangent=-0.2554, interpMode=RCIM_Cubic, leaveTangent=-0.2554, tangentMode=RCTM_User, time=0.4, value=0.97}
       <at 1 0>
      ]}}
   }
  }
 }
 {id=Mixer_0, children [Attenuation_1, Attenuation_0, Attenuation_2], properties {inputVolume [4,4,2]}}
 {id=WavePlayer_0, properties {bLooping=true, soundWaveAssetPtr=/Game/Athena/Sounds/Water/Ambient/Amb_Water_Waterfall_Large_Bot_Loop_01}}
 {id=WavePlayer_1, properties {bLooping=true, soundWaveAssetPtr=/Game/Athena/Sounds/Water/Ambient/Amb_Water_Waterfall_Med_Top_Loop_01}}
]
properties {
 bOverrideConcurrency=true
 maxDistance=8500
 soundClassObject=/Game/Sounds/FortSoundClassesAndModes/Fort_Water_Ambient
 volumeMultiplier=0.8
 attenuationOverrides {
  bEnableListenerFocus=true
  distanceAlgorithm=Logarithmic
  falloffDistance=5000
  focusAzimuth=45
  focusVolumeAttenuation=1.2
  lPFFrequencyAtMax=1000
  nonFocusVolumeAttenuation=0.9
  nonSpatializedRadiusStart=700
  stereoSpread=1000
  attenuationShapeExtents {x=600}
 }
 concurrencyOverrides {maxCount=2, voiceStealReleaseTime=5}
 concurrencySet [/Game/Sounds/Concurrency_Rules/Waterfall_StopFarthestOldest_Conc]
}
