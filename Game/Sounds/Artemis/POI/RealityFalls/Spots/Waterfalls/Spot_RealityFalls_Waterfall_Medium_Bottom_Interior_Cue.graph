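# Looping ambient spot sound cue for the interior bottom of a medium waterfall at Reality Falls, on the water ambient sound class.
# The first quality input mixes exterior, close interior, white-water and pitched-up distant roar loops.
# The other three quality inputs share a single pitched-up close roar loop.
# Per-layer attenuation nodes give separate falloff and listener focus, with occlusion on only the two roar layers, and concurrency is capped.
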SoundCue /Game/Sounds/Artemis/POI/RealityFalls/Spots/Waterfalls/Spot_RealityFalls_Waterfall_Medium_Bottom_Interior_Cue
root=QualityLevel_0
nodes [
 {
  id=Attenuation_2
  children [Modulator_4]
  properties {
   bOverrideAttenuation=true
   attenuationOverrides {
    bEnableFocusInterpolation=true
    bEnableListenerFocus=true
    bEnableOcclusion=true
    distanceAlgorithm=Custom
    falloffDistance=7500
    lPFFrequencyAtMax=1000
    lPFRadiusMax=4000
    lPFRadiusMin=1000
    nonFocusVolumeAttenuation=0.8
    nonSpatializedRadiusStart=5000
    occlusionInterpolationTime=3
    occlusionLowPassFilterFrequency=1000
    occlusionTraceChannel=ECC_GameTraceChannel11
    reverbWetLevelMax=1.2
    reverbWetLevelMin=0.5
    stereoSpread=10000
    attenuationShapeExtents {x=2000}
    customAttenuationCurve {editorCurveData {keys [
       <at 0 0>
       {arriveTangent=0.028, interpMode=RCIM_Cubic, leaveTangent=0.028, tangentMode=RCTM_User, time=0.2209, value=1}
       {arriveTangent=-0.2554, interpMode=RCIM_Cubic, leaveTangent=-0.2554, tangentMode=RCTM_User, time=0.4, value=0.97}
       <at 1 0>
      ]}}
   }
  }
 }
 {
  id=Attenuation_3
  children [Modulator_0]
  properties {
   bOverrideAttenuation=true
   attenuationOverrides {
    bAttenuateWithLPF=true
    bEnableFocusInterpolation=true
    bEnableListenerFocus=true
    bEnableOcclusion=true
    distanceAlgorithm=Logarithmic
    falloffDistance=7500
    lPFFrequencyAtMax=1000
    lPFRadiusMax=4000
    lPFRadiusMin=1000
    nonFocusVolumeAttenuation=0.8
    nonSpatializedRadiusStart=3000
    occlusionInterpolationTime=2
    occlusionLowPassFilterFrequency=1000
    occlusionTraceChannel=ECC_GameTraceChannel11
    reverbWetLevelMax=1.2
    reverbWetLevelMin=0.2
    stereoSpread=500
    attenuationShapeExtents {x=2000}
   }
  }
 }
 {
  id=Attenuation_4
  children [Mixer_3]
  properties {
   bOverrideAttenuation=true
   attenuationOverrides {
    bAttenuateWithLPF=true
    bEnableFocusInterpolation=true
    bEnableListenerFocus=true
    distanceAlgorithm=NaturalSound
    falloffDistance=1500
    focusAttackInterpSpeed=0.1
    focusAzimuth=35
    lPFFrequencyAtMax=1000
    lPFRadiusMax=4000
    lPFRadiusMin=1000
    nonFocusAzimuth=65
    nonFocusVolumeAttenuation=0.5
    nonSpatializedRadiusStart=500
    occlusionInterpolationTime=2
    occlusionLowPassFilterFrequency=1000
    occlusionTraceChannel=ECC_GameTraceChannel11
    reverbWetLevelMax=1.2
    reverbWetLevelMin=0.6
    stereoSpread=2000
    attenuationShapeExtents {x=500}
   }
  }
 }
 {
  id=Attenuation_5
  children [WavePlayer_1]
  properties {
   bOverrideAttenuation=true
   attenuationOverrides {
    bAttenuateWithLPF=true
    bEnableFocusInterpolation=true
    bEnableListenerFocus=true
    distanceAlgorithm=Logarithmic
    falloffDistance=2000
    lPFFrequencyAtMax=1000
    lPFRadiusMax=4000
    lPFRadiusMin=1000
    nonFocusVolumeAttenuation=0.8
    nonSpatializedRadiusStart=500
    occlusionInterpolationTime=2
    occlusionLowPassFilterFrequency=1000
    occlusionTraceChannel=ECC_GameTraceChannel11
    reverbWetLevelMax=1.2
    reverbWetLevelMin=0.5
    attenuationShapeExtents {x=750}
   }
  }
 }
 {
  id=Attenuation_7
  children [Mixer_1]
  properties {
   bOverrideAttenuation=true
   attenuationOverrides {
    bAttenuateWithLPF=true
    bEnableFocusInterpolation=true
    bEnableListenerFocus=true
    distanceAlgorithm=Logarithmic
    falloffDistance=2500
    focusVolumeAttenuation=1.2
    lPFFrequencyAtMax=1000
    lPFRadiusMax=4000
    lPFRadiusMin=1000
    nonFocusVolumeAttenuation=0.9
    nonSpatializedRadiusStart=750
    occlusionInterpolationTime=1
    occlusionLowPassFilterFrequency=3000
    occlusionTraceChannel=ECC_GameTraceChannel11
    occlusionVolumeAttenuation=0.8
    reverbWetLevelMax=1.2
    reverbWetLevelMin=0.5
    stereoSpread=450
    attenuationShapeExtents {x=750}
   }
  }
 }
 {id=Mixer_0, children [Attenuation_7, Mixer_2, Attenuation_2], properties {inputVolume [0.5,1,1]}}
 {id=Mixer_1, children [WavePlayer_6], properties {inputVolume [1]}}
 {id=Mixer_2, children [Attenuation_5, Attenuation_4], properties {inputVolume [0.75,3]}}
 {id=Mixer_3, children [WavePlayer_7,WavePlayer_0], properties {inputVolume [1,2]}}
 {id=Modulator_0, children [WavePlayer_4], properties {pitchMax=1.3, pitchMin=1.2}}
 {id=Modulator_4, children [WavePlayer_3], properties {pitchMax=1.3, pitchMin=1.2}}
 {id=QualityLevel_0, children [Mixer_0, Attenuation_3, Attenuation_3, Attenuation_3]}
 {
  id=WavePlayer_0
  properties {
   bLooping=true
   soundWaveAssetPtr=&/Ambience_RealityFalls_Waterfall_Rolling_White_Water_Loop
  }
 }
 {
  id=WavePlayer_1
  properties {
   bLooping=true
   soundWaveAssetPtr=&/Ambience_RealityFalls_Waterfall_Medium_Bottom_Interior_Close_Loop
  }
 }
 {
  id=WavePlayer_3
  properties {bLooping=true, soundWaveAssetPtr=&/Spot_RealityFalls_Waterfall_Roar_Medium_Distant_01}
 }
 {
  id=WavePlayer_4
  properties {bLooping=true, soundWaveAssetPtr=&/Spot_RealityFalls_Waterfall_Roar_Close}
 }
 {
  id=WavePlayer_6
  properties {
   bLooping=true
   soundWaveAssetPtr=&/Ambience_RealityFalls_Waterfall_Medium_Bottom_Exterior_Loop
  }
 }
 {
  id=WavePlayer_7
  properties {
   bLooping=true
   soundWaveAssetPtr=&Ambience_RealityFalls_Waterfall_Medium_Bottom_Interior_Close_Loop
  }
 }
]
properties {
 bDebug=true
 bOverrideConcurrency=true
 maxDistance=9500
 soundClassObject=/Game/Sounds/FortSoundClassesAndModes/Fort_Water_Ambient
 volumeMultiplier=0.5
 attenuationOverrides {
  distanceAlgorithm=Logarithmic
  falloffDistance=2000
  lPFFrequencyAtMax=500
  lPFRadiusMax=2500
  lPFRadiusMin=500
  attenuationShapeExtents {x=500}
 }
 concurrencyOverrides {maxCount=3}
}
