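# Looping spot ambience cue for a medium waterfall spray at Reality Falls, built from Dam_Waterfall waves and limited to one concurrent instance.
# A mixer blends one of three random close loops with a half-volume distant loop, each through its own occlusion- and focus-aware attenuation.
# The close layer uses logarithmic falloff with distance low-pass.
# The distant layer's custom curve is silent at the source, peaks farther out, then fades, with much wider stereo spread.
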SoundCue /Game/Sounds/Artemis/POI/RealityFalls/Spots/Waterfalls/Spot_RealityFalls_Waterfall_Medium_Spray_Cue
root=Mixer_0
nodes [
 {
  id=Attenuation_0
  children [Random_0]
  properties {
   bOverrideAttenuation=true
   attenuationOverrides {
    bAttenuateWithLPF=true
    bEnableFocusInterpolation=true
    bEnableListenerFocus=true
    bEnableOcclusion=true
    distanceAlgorithm=Logarithmic
    falloffDistance=7500
    lPFFrequencyAtMax=1000
    lPFRadiusMax=4000
    lPFRadiusMin=1000
    nonFocusVolumeAttenuation=0.8
    nonSpatializedRadiusStart=3000
    occlusionInterpolationTime=2
    occlusionLowPassFilterFrequency=1000
    occlusionTraceChannel=ECC_GameTraceChannel11
    stereoSpread=500
    attenuationShapeExtents {x=2000}
   }
  }
 }
 {
  id=Attenuation_2
  children [WavePlayer_0]
  properties {
   bOverrideAttenuation=true
   attenuationOverrides {
    bEnableFocusInterpolation=true
    bEnableListenerFocus=true
    bEnableOcclusion=true
    distanceAlgorithm=Custom
    falloffDistance=7500
    lPFFrequencyAtMax=1000
    lPFRadiusMax=4000
    lPFRadiusMin=1000
    nonFocusVolumeAttenuation=0.8
    nonSpatializedRadiusStart=5000
    occlusionInterpolationTime=3
    occlusionLowPassFilterFrequency=1000
    occlusionTraceChannel=ECC_GameTraceChannel11
    stereoSpread=10000
    attenuationShapeExtents {x=2000}
    customAttenuationCurve {editorCurveData {keys [
       <at 0 0>
       {arriveTangent=0.028, interpMode=RCIM_Cubic, leaveTangent=0.028, tangentMode=RCTM_User, time=0.2209, value=1}
       {arriveTangent=-0.2554, interpMode=RCIM_Cubic, leaveTangent=-0.2554, tangentMode=RCTM_User, time=0.4, value=0.97}
       <at 1 0>
      ]}}
   }
  }
 }
 {id=Mixer_0, children [Attenuation_0, Attenuation_2], properties {inputVolume [1,0.5]}}
 {id=Random_0, children [WavePlayer_3,WavePlayer_4,WavePlayer_5], properties {weights [1,1,1]}}
 {
  id=WavePlayer_0
  properties {bLooping=true, soundWaveAssetPtr=/Game/Sounds/Fort_Prop_Sounds/Exterior/Dam_Waterfall/Dam_Waterfall_Base_Distant_01}
 }
 {
  id=WavePlayer_3
  properties {bLooping=true, soundWaveAssetPtr=/Game/Sounds/Fort_Prop_Sounds/Exterior/Dam_Waterfall/Dam_Waterfall_Base_Close_01a}
 }
 {
  id=WavePlayer_4
  properties {bLooping=true, soundWaveAssetPtr=/Game/Sounds/Fort_Prop_Sounds/Exterior/Dam_Waterfall/Dam_Waterfall_Base_Close_01b}
 }
 {
  id=WavePlayer_5
  properties {bLooping=true, soundWaveAssetPtr=/Game/Sounds/Fort_Prop_Sounds/Exterior/Dam_Waterfall/Dam_Waterfall_Base_Close_01c}
 }
]
properties {
 attenuationSettings=/Game/Sounds/Attenuation/Fort_Ambient_Capsule_Waterfall_Smaller_Att
 bOverrideConcurrency=true
 maxDistance=9500
 soundClassObject=/Game/Sounds/FortSoundClassesAndModes/Fort_Water_Ambient
 volumeMultiplier=1
 attenuationOverrides {
  bAttenuateWithLPF=true
  bEnableFocusInterpolation=true
  bEnableListenerFocus=true
  bEnableOcclusion=true
  falloffDistance=7500
  lPFFrequencyAtMax=1000
  lPFRadiusMax=4000
  lPFRadiusMin=1000
  nonFocusVolumeAttenuation=0.6
  nonSpatializedRadiusStart=3000
  occlusionInterpolationTime=2
  occlusionLowPassFilterFrequency=1000
  occlusionTraceChannel=ECC_GameTraceChannel11
  stereoSpread=500
  attenuationShapeExtents {x=2000}
 }
 concurrencyOverrides {maxCount=1}
}
