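# Sound cue for a waterfall pool ambient spot at the Reality Falls POI, named for that location but holding no sound nodes or root, so it has no audio to play as it stands.
# It keeps playback settings: the outdoor ambient sound class, reduced volume, and a shared river water attenuation asset alongside its own stored attenuation values.
# A concurrency override limits it to one voice, with a release time for stolen voices.
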SoundCue /Game/Sounds/Artemis/POI/RealityFalls/Spots/Waterfalls/Spot_RealityFalls_Waterfall_Pool_Cue
root=null
nodes []
properties {
 attenuationSettings=/Game/Athena/Sounds/Ambient/Fort_Amb_Water_River_Att
 bOverrideConcurrency=true
 maxDistance=2925
 soundClassObject=&Fort_Ambient_Outdoors_Athena
 volumeMultiplier=0.601
 attenuationOverrides {falloffDistance=2000, nonSpatializedRadiusStart=700, stereoSpread=1000, attenuationShapeExtents {x=925}}
 concurrencyOverrides {maxCount=1, voiceStealReleaseTime=2}
}
